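# One-shot mono MetaSound Source for the Specter pawn's appear sound, named as the 1P variant.
# On play, a very short trigger delay fires a random get from an array of appear waves.
# The chosen wave plays through a wave player and mixer to the output, and its completion ends the source.
# It uses Specter pawn attenuation, the in-game SFX class and a global ghost concurrency set.
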MetaSoundSource /SpecterGameplay/Audio/Gameplay/MSS_SpecterPawn_Appear_1P
inputs [
 ghost appear and vocal {type=WaveAsset:Array}
 UE.Source.OnPlay {type=Trigger}
]
interfaces [UE.OutputFormat.Mono:1.0, UE.Source.OneShot:1.0, UE.Source:1.1]
outputs [
 "UE.OutputFormat.Mono.Audio:0" {type=Audio}
 UE.Source.OneShot.OnFinished {type=Trigger}
]
pages [Default {
  input_defaults {
   UE.Source.OnPlay=null
   ghost appear and vocal [
    /SpecterGameplay/Audio/Gameplay/Waves/Delulu_SpecterPawn_Appear_Mono_01
    /SpecterGameplay/Audio/Gameplay/Waves/Delulu_SpecterPawn_Appear_Mono_02
    /SpecterGameplay/Audio/Gameplay/Waves/Delulu_SpecterPawn_Appear_Mono_03
    /SpecterGameplay/Audio/Gameplay/Waves/Delulu_SpecterPawn_Appear_Mono_04
    /SpecterGameplay/Audio/Gameplay/Waves/Delulu_SpecterPawn_Appear_Mono_05
    /SpecterGameplay/Audio/Gameplay/Waves/Delulu_SpecterPawn_Appear_Mono_06
    /SpecterGameplay/Audio/Gameplay/Waves/Delulu_SpecterPawn_Appear_Mono_07
    /SpecterGameplay/Audio/Gameplay/Waves/Delulu_SpecterPawn_Appear_Mono_08
    /SpecterGameplay/Audio/Gameplay/Waves/Delulu_SpecterPawn_Appear_Mono_09
    /SpecterGameplay/Audio/Gameplay/Waves/Delulu_SpecterPawn_Appear_Mono_10
    /SpecterGameplay/Audio/Gameplay/Waves/Delulu_SpecterPawn_Appear_Mono_11
    /SpecterGameplay/Audio/Gameplay/Waves/Delulu_SpecterPawn_Appear_Mono_12
    /SpecterGameplay/Audio/Gameplay/Waves/Delulu_SpecterPawn_Appear_Mono_13
   ]
  }
  nodes [
   TriggerDelay_0 = UE.Trigger Delay(In=^UE.Source.OnPlay) {also {defaults {Delay Time=0.001}}}
   ArrayRandomGet_0 = Array.Random Get:WaveAsset:Array(In Array=^ghost appear and vocal, Next=@TriggerDelay_0.Out)
   WavePlayer_0 = UE.Wave Player:Mono(Play=@ArrayRandomGet_0.On Next, Wave Asset=@ArrayRandomGet_0.Value) {also {version [1,1]}}
   AudioMixerAudioMixerMono2_0 = "AudioMixer.Audio Mixer (Mono, 2)"(In 0=@WavePlayer_0.Out Mono)
  ]
  links [
   {dst="@UE.OutputFormat.Mono.Audio:0", src=AudioMixerAudioMixerMono2_0.Out}
   {dst="@UE.Source.OneShot.OnFinished", src=WavePlayer_0.On Finished}
  ]
 }]
properties {
 assetImportData=&self:AssetImportData
 attenuationSettings=/SpecterGameplay/Audio/Data/ATT_Specter_Pawn
 maxDistance=1050
 sampleRate=48000
 soundClassObject=&Fort_SFX_InGame
 volume=3
 concurrencySet [&CONC_Delulu_Specter_Ghost_Global]
}
