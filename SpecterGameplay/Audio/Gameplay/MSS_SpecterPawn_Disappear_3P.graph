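# Stereo one-shot MetaSound for a Specter pawn disappearing, named as the third-person variant, with Specter attenuation and concurrency.
# On play, it fires two layers together: a random idle twinkle from three waves and a fixed disappear pop.
# Each layer gets its own random pitch shift, and a stereo mixer blends them with the pop slightly lowered.
# The sound finishes when the twinkle player ends, regardless of the pop.
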MetaSoundSource /SpecterGameplay/Audio/Gameplay/MSS_SpecterPawn_Disappear_3P
inputs [
 In Array {type=WaveAsset:Array}
 UE.Source.OnPlay {type=Trigger}
]
interfaces [UE.OutputFormat.Stereo:1.0, UE.Source.OneShot:1.0, UE.Source:1.1]
outputs [
 "UE.OutputFormat.Stereo.Audio:0" {type=Audio}
 "UE.OutputFormat.Stereo.Audio:1" {type=Audio}
 UE.Source.OneShot.OnFinished {type=Trigger}
]
pages [Default {
  input_defaults {
   UE.Source.OnPlay=null
   In Array [
    /SpecterGameplay/Audio/Gameplay/Waves/Delulu_SpecterPawn_Idle_Twinkle_03
    /SpecterGameplay/Audio/Gameplay/Waves/Delulu_SpecterPawn_Idle_Twinkle_05
    /SpecterGameplay/Audio/Gameplay/Waves/Delulu_SpecterPawn_Idle_Twinkle_06
   ]
  }
  nodes [
   n1 = UE.RandomFloat(Next=^UE.Source.OnPlay) {also {defaults {Max=1.25, Min=-0.75}}}
   ArrayRandomGet_0 = Array.Random Get:WaveAsset:Array(In Array=^In Array, Next=@n1.On Next)
   WavePlayer_0 = UE.Wave Player:Stereo(Play=@ArrayRandomGet_0.On Next, Wave Asset=@ArrayRandomGet_0.Value, Pitch Shift=@n1.Value) {also {version [1,1]}}
   n2 = UE.RandomFloat(Next=^UE.Source.OnPlay) {also {defaults {Max=1.25, Min=-0.75}}}
   WavePlayer_1 = UE.Wave Player:Stereo(Play=@n2.On Next, Pitch Shift=@n2.Value) {also {defaults {Wave Asset=/SpecterGameplay/Audio/Gameplay/Waves/Delulu_SpecterPawn_Disappear_Pop}, version [1,1]}}
   AudioMixerAudioMixerStereo4_0 = "AudioMixer.Audio Mixer (Stereo, 4)"(In 0 L=@WavePlayer_0.Out Left, In 0 R=@WavePlayer_0.Out Right, In 1 L=@WavePlayer_1.Out Left, In 1 R=@WavePlayer_1.Out Right) {also {defaults {Gain 1=0.8}}}
  ]
  links [
   {dst="@UE.OutputFormat.Stereo.Audio:0", src=AudioMixerAudioMixerStereo4_0.Out L}
   {dst="@UE.OutputFormat.Stereo.Audio:1", src=AudioMixerAudioMixerStereo4_0.Out R}
   {dst="@UE.Source.OneShot.OnFinished", src=WavePlayer_0.On Finished}
  ]
 }]
properties {
 assetImportData=&self:AssetImportData
 attenuationSettings=/SpecterGameplay/Audio/Data/ATT_Specter_Pawn
 maxDistance=1050
 numChannels=2
 outputFormat=Stereo
 sampleRate=48000
 soundClassObject=&Fort_SFX_InGame
 concurrencySet [&CONC_Delulu_Specter_Ghost_Global]
}
