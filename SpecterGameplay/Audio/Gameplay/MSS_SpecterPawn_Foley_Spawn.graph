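# One-shot stereo MetaSound that plays the Specter pawn's spawn foley, a woosh with a sparkle layer, using Specter pawn attenuation and concurrency.
# On play, a random sparkle from an array and a fixed woosh start together, then mix through per-layer gain inputs.
# The woosh player references its wave directly, leaving the Woosh array input unconnected.
# The sparkle player alone fires OnFinished, so the sound ends with that layer.
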MetaSoundSource /SpecterGameplay/Audio/Gameplay/MSS_SpecterPawn_Foley_Spawn
inputs [
 Sparkle {type=WaveAsset:Array}
 Sparkle_Gain {type=Float}
 UE.Source.OnPlay {type=Trigger}
 Woosh {type=WaveAsset:Array}
 Woosh_Gain {type=Float}
]
interfaces [UE.OutputFormat.Stereo:1.0, UE.Source.OneShot:1.0, UE.Source:1.1]
outputs [
 "UE.OutputFormat.Stereo.Audio:0" {type=Audio}
 "UE.OutputFormat.Stereo.Audio:1" {type=Audio}
 UE.Source.OneShot.OnFinished {type=Trigger}
]
pages [Default {
  input_defaults {
   Sparkle_Gain=0.28
   UE.Source.OnPlay=null
   Woosh_Gain=1
   Sparkle [
    /SpecterGameplay/Audio/Gameplay/Waves/Delulu_SpecterPawn_Spawn_Sparkle_01
    /SpecterGameplay/Audio/Gameplay/Waves/Delulu_SpecterPawn_Spawn_Sparkle_02
    /SpecterGameplay/Audio/Gameplay/Waves/Delulu_SpecterPawn_Spawn_Sparkle_03
   ]
   Woosh [/SpecterGameplay/Audio/Gameplay/Waves/Delulu_SpecterPawn_Spawn_Woosh_01]
  }
  nodes [
   ArrayRandomGet_0 = Array.Random Get:WaveAsset:Array(In Array=^Sparkle, Next=^UE.Source.OnPlay) {also {defaults {Weights []}}}
   WavePlayer_0 = UE.Wave Player:Stereo(Play=@ArrayRandomGet_0.On Next, Wave Asset=@ArrayRandomGet_0.Value) {also {version [1,1]}}
   WavePlayer_1 = UE.Wave Player:Stereo(Play=^UE.Source.OnPlay) {also {defaults {Wave Asset=&Delulu_SpecterPawn_Spawn_Woosh_01}, version [1,1]}}
   AudioMixerAudioMixerStereo2_0 = "AudioMixer.Audio Mixer (Stereo, 2)"(
    Gain 0=^Sparkle_Gain
    Gain 1=^Woosh_Gain
    In 0 L=@WavePlayer_0.Out Left
    In 0 R=@WavePlayer_0.Out Right
    In 1 L=@WavePlayer_1.Out Left
    In 1 R=@WavePlayer_1.Out Right
   )
  ]
  links [
   {dst="@UE.OutputFormat.Stereo.Audio:0", src=AudioMixerAudioMixerStereo2_0.Out L}
   {dst="@UE.OutputFormat.Stereo.Audio:1", src=AudioMixerAudioMixerStereo2_0.Out R}
   {dst="@UE.Source.OneShot.OnFinished", src=WavePlayer_0.On Finished}
  ]
 }]
properties {
 assetImportData=&self:AssetImportData
 attenuationSettings=/SpecterGameplay/Audio/Data/ATT_Specter_Pawn
 maxDistance=1050
 numChannels=2
 outputFormat=Stereo
 sampleRate=48000
 soundClassObject=&Fort_SFX_InGame
 volume=0.8
 concurrencySet [&CONC_Delulu_Specter_Ghost_Global]
}
