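# One-shot stereo MetaSound for the first-person Specter pawn jump, with attenuation and a global ghost concurrency set.
# On play, it picks a random bubble wave and a random jump wave, playing both with a shared random pitch shift.
# The bubble layer is scaled by the Gain input and mixed with the jump layer, which the mixer attenuates.
# The bubble player alone triggers OnFinished, so its length determines when the sound ends.
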MetaSoundSource /SpecterGameplay/Audio/Gameplay/MSS_SpecterPawn_Jump_1P
inputs [
 Bubbles {type=WaveAsset:Array}
 Gain {type=Float}
 Jump {type=WaveAsset:Array}
 UE.Source.OnPlay {type=Trigger}
]
interfaces [UE.OutputFormat.Stereo:1.0, UE.Source.OneShot:1.0, UE.Source:1.1]
outputs [
 "UE.OutputFormat.Stereo.Audio:0" {type=Audio}
 "UE.OutputFormat.Stereo.Audio:1" {type=Audio}
 UE.Source.OneShot.OnFinished {type=Trigger}
]
pages [Default {
  input_defaults {
   Gain=0.58
   UE.Source.OnPlay=null
   Bubbles [
    /SpecterGameplay/Audio/Gameplay/Waves/Delulu_SpecterPawn_Idle_Bubble_01
    /SpecterGameplay/Audio/Gameplay/Waves/Delulu_SpecterPawn_Idle_Bubble_02
    /SpecterGameplay/Audio/Gameplay/Waves/Delulu_SpecterPawn_Idle_Bubble_03
    /SpecterGameplay/Audio/Gameplay/Waves/Delulu_SpecterPawn_Idle_Bubble_04
    /SpecterGameplay/Audio/Gameplay/Waves/Delulu_SpecterPawn_Idle_Bubble_05
    /SpecterGameplay/Audio/Gameplay/Waves/Delulu_SpecterPawn_Idle_Bubble_06
    /SpecterGameplay/Audio/Gameplay/Waves/Delulu_SpecterPawn_Idle_Bubble_07
    /SpecterGameplay/Audio/Gameplay/Waves/Delulu_SpecterPawn_Idle_Bubble_08
   ]
   Jump [
    /SpecterGameplay/Audio/Gameplay/Waves/Delulu_SpecterPawn_Jump_01
    /SpecterGameplay/Audio/Gameplay/Waves/Delulu_SpecterPawn_Jump_02
    /SpecterGameplay/Audio/Gameplay/Waves/Delulu_SpecterPawn_Jump_03
   ]
  }
  nodes [
   ArrayRandomGet_1 = Array.Random Get:WaveAsset:Array(In Array=^Bubbles, Next=^UE.Source.OnPlay)
   n1 = UE.RandomFloat(Next=^UE.Source.OnPlay)
   WavePlayer_1 = UE.Wave Player:Stereo(Play=@ArrayRandomGet_1.On Next, Wave Asset=@ArrayRandomGet_1.Value, Pitch Shift=@n1.Value) {also {version [1,1]}}
   ArrayRandomGet_0 = Array.Random Get:WaveAsset:Array(In Array=^Jump, Next=^UE.Source.OnPlay) {also {defaults {Weights []}}}
   WavePlayer_0 = UE.Wave Player:Stereo(Play=@ArrayRandomGet_0.On Next, Wave Asset=@ArrayRandomGet_0.Value, Pitch Shift=@n1.Value) {also {version [1,1]}}
   AudioMixerAudioMixerStereo4_0 = "AudioMixer.Audio Mixer (Stereo, 4)"(
    In 0 R=UE.Multiply:Audio by Float(AdditionalOperands=^Gain, PrimaryOperand=@WavePlayer_1.Out Right).Out
    In 0 L=UE.Multiply:Audio by Float(AdditionalOperands=^Gain, PrimaryOperand=@WavePlayer_1.Out Left).Out
    In 1 L=@WavePlayer_0.Out Left
    In 1 R=@WavePlayer_0.Out Right
   ) {also {defaults {Gain 1=0.5}}}
  ]
  links [
   {dst="@UE.OutputFormat.Stereo.Audio:0", src=AudioMixerAudioMixerStereo4_0.Out L}
   {dst="@UE.OutputFormat.Stereo.Audio:1", src=AudioMixerAudioMixerStereo4_0.Out R}
   {dst="@UE.Source.OneShot.OnFinished", src=WavePlayer_1.On Finished}
  ]
 }]
properties {
 assetImportData=&self:AssetImportData
 attenuationSettings=/SpecterGameplay/Audio/Data/ATT_Specter_Pawn
 maxDistance=1050
 numChannels=2
 outputFormat=Stereo
 sampleRate=48000
 soundClassObject=&Fort_SFX_InGame
 volume=4
 concurrencySet [&CONC_Delulu_Specter_Ghost_Global]
}
