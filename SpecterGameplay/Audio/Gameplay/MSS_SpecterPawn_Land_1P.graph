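# One-shot stereo MetaSound Source for the Specter pawn landing sound, named as the 1P variant.
# On play, it picks a random wave from the landing array and plays it through a stereo Wave Player with a random upward pitch shift.
# The player's finish trigger ends the source; the Gain input is not connected to anything.
# Specter pawn attenuation and a global ghost concurrency set govern playback.
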MetaSoundSource /SpecterGameplay/Audio/Gameplay/MSS_SpecterPawn_Land_1P
inputs [
 Gain {type=Float}
 In Array {type=WaveAsset:Array}
 UE.Source.OnPlay {type=Trigger}
]
interfaces [UE.OutputFormat.Stereo:1.0, UE.Source.OneShot:1.0, UE.Source:1.1]
outputs [
 "UE.OutputFormat.Stereo.Audio:0" {type=Audio}
 "UE.OutputFormat.Stereo.Audio:1" {type=Audio}
 UE.Source.OneShot.OnFinished {type=Trigger}
]
pages [Default {
  input_defaults {
   Gain=0.58
   UE.Source.OnPlay=null
   In Array [/SpecterGameplay/Audio/Gameplay/Waves/Delulu_SpecterPawn_Landing_01, /SpecterGameplay/Audio/Gameplay/Waves/Delulu_SpecterPawn_Landing_02]
  }
  nodes [
   ArrayRandomGet_0 = Array.Random Get:WaveAsset:Array(In Array=^In Array, Next=^UE.Source.OnPlay)
   WavePlayer_0 = UE.Wave Player:Stereo(
    Play=@ArrayRandomGet_0.On Next
    Wave Asset=@ArrayRandomGet_0.Value
    Pitch Shift=UE.RandomFloat(Next=^UE.Source.OnPlay) {also {defaults {Max=1.5, Min=1}}}.Value
   ) {also {version [1,1]}}
  ]
  links [
   {dst="@UE.Source.OneShot.OnFinished", src=WavePlayer_0.On Finished}
   {dst="@UE.OutputFormat.Stereo.Audio:0", src=WavePlayer_0.Out Left}
   {dst="@UE.OutputFormat.Stereo.Audio:1", src=WavePlayer_0.Out Right}
  ]
 }]
properties {
 assetImportData=&self:AssetImportData
 attenuationSettings=/SpecterGameplay/Audio/Data/ATT_Specter_Pawn
 maxDistance=1050
 numChannels=2
 outputFormat=Stereo
 sampleRate=48000
 soundClassObject=&Fort_SFX_InGame
 concurrencySet [&CONC_Delulu_Specter_Ghost_Global]
}
