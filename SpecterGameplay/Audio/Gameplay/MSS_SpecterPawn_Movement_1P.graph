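# Mono first-person movement MetaSound for the Specter pawn, blending a speed-driven movement wave with random idle bubble and bamboo sounds.
# Smoothed Speed sets the movement wave's pitch and LFO amplitude-modulation rate, which slows as speed rises, while raw Speed raises its volume.
# After a startup delay, nested repeat triggers play random idle waves at randomized intervals and pitch, fading out quickly once Speed rises.
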MetaSoundSource /SpecterGameplay/Audio/Gameplay/MSS_SpecterPawn_Movement_1P
inputs [
 Idle Bubbles and Bamboo {type=WaveAsset:Array}
 Speed {type=Float}
 UE.Source.OnPlay {type=Trigger}
]
interfaces [UE.OutputFormat.Mono:1.0, UE.Source.OneShot:1.0, UE.Source:1.1]
outputs [
 "UE.OutputFormat.Mono.Audio:0" {type=Audio}
 UE.Source.OneShot.OnFinished {type=Trigger}
]
pages [Default {
  input_defaults {
   Speed=0
   UE.Source.OnPlay=null
   Idle Bubbles and Bamboo [
    /SpecterGameplay/Audio/Gameplay/Waves/Delulu_SpecterPawn_Idle_Bamboo_Mono_01
    /SpecterGameplay/Audio/Gameplay/Waves/Delulu_SpecterPawn_Idle_Bamboo_Mono_02
    /SpecterGameplay/Audio/Gameplay/Waves/Delulu_SpecterPawn_Idle_Bamboo_Mono_03
    /SpecterGameplay/Audio/Gameplay/Waves/Delulu_SpecterPawn_Idle_Bubble_mono_02
    &Delulu_SpecterPawn_Idle_Bubble_mono_02
    /SpecterGameplay/Audio/Gameplay/Waves/Delulu_SpecterPawn_Idle_Bubble_mono_03
   ]
  }
  nodes [
   n1 = UE.InterpTo:Audio(Target=^Speed) {also {defaults {Interp Time=0.08}}}
   n2 = UE.LFO:Audio(Frequency=MapRange.MapRange:Float(In=@n1.Value) {also {defaults {In Range A=0.5, In Range B=0.8, Out Range A=8, Out Range B=0.5}}}.Out Value) {also {defaults {Min Value=0.5, Shape=2}}}
   WavePlayer_1 = UE.Wave Player:Mono(
    Play=^UE.Source.OnPlay
    Pitch Shift=MapRange.MapRange:Float(In=@n1.Value) {also {defaults {In Range A=0.3, In Range B=0.8, Out Range B=3}}}.Out Value
   ) {also {
     defaults {Loop="<MetaSoundBuilderResult.SUCCEEDED: 0>", Wave Asset=/SpecterGameplay/Audio/Gameplay/Waves/Delulu_Specter_Movement_loop_02}
     version [1,1]
    }}
   TriggerDelay_0 = UE.Trigger Delay(In=^UE.Source.OnPlay) {also {defaults {Delay Time=3}}}
   n3 = UE.TriggerRepeat(Start=@TriggerDelay_0.Out) {also {defaults {Period=60}}}
   n4 = UE.RandomTime(Next=@n3.RepeatOut) {also {defaults {Max=14, Min=10}}}
   ArrayRandomGet_0 = Array.Random Get:WaveAsset:Array(In Array=^Idle Bubbles and Bamboo, Next=UE.TriggerRepeat(Start=@n4.On Next, Period=@n4.Value).RepeatOut) {also {defaults {Weights []}}}
   WavePlayer_0 = UE.Wave Player:Mono(
    Play=@ArrayRandomGet_0.On Next
    Wave Asset=@ArrayRandomGet_0.Value
    Pitch Shift=UE.RandomFloat(Next=@n3.RepeatOut) {also {defaults {Max=2, Min=1}}}.Value
   ) {also {version [1,1]}}
   AudioMixerAudioMixerMono3_0 = "AudioMixer.Audio Mixer (Mono, 3)"(
    Gain 2=MapRange.MapRange:Float(In=^Speed) {also {defaults {In Range B=0.08, Out Range A=0.8, Out Range B=0}}}.Out Value
    In 0=UE.Multiply:Audio by Float(
     AdditionalOperands=MapRange.MapRange:Float(In=^Speed) {also {defaults {In Range B=0.3, Out Range A=0.1, Out Range B=0.7}}}.Out Value
     PrimaryOperand=UE.Multiply:Audio by Float(AdditionalOperands=@n2.Out, PrimaryOperand=@WavePlayer_1.Out Mono).Out
    ).Out
    In 2=@WavePlayer_0.Out Mono
   ) {also {defaults {Gain 0=2}}}
  ]
  links [
   {dst="@UE.OutputFormat.Mono.Audio:0", src=AudioMixerAudioMixerMono3_0.Out}
   {dst="@UE.Source.OneShot.OnFinished", src=WavePlayer_1.On Finished}
  ]
 }]
properties {
 assetImportData=&self:AssetImportData
 attenuationSettings=/SpecterGameplay/Audio/Data/ATT_Specter_Pawn
 maxDistance=1050
 sampleRate=48000
 soundClassObject=&Fort_SFX_InGame
 concurrencySet [&CONC_Delulu_Specter_Ghost_Global]
}
