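# Stereo one-shot MetaSound for firing a Specter "doot", layering a random cloud poof with a random ghost vocalization.
# On play, the poof starts directly, while the vocal passes through a zero-time delay and single repeat, then gets a randomized upward pitch shift.
# A stereo mixer combines both, with the vocal attenuated.
# Only the poof player's finish signals OnFinished, so the poof sets the sound's length.
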MetaSoundSource /SpecterGameplay/Audio/Gameplay/MSS_Specter_Doot_Fire
inputs [
 ghost vocal {type=WaveAsset:Array}
 poof {type=WaveAsset:Array}
 UE.Source.OnPlay {type=Trigger}
]
interfaces [UE.OutputFormat.Stereo:1.0, UE.Source.OneShot:1.0, UE.Source:1.1]
outputs [
 "UE.OutputFormat.Stereo.Audio:0" {type=Audio}
 "UE.OutputFormat.Stereo.Audio:1" {type=Audio}
 UE.Source.OneShot.OnFinished {type=Trigger}
]
pages [Default {
  input_defaults {
   UE.Source.OnPlay=null
   ghost vocal [
    /SpecterGameplay/Audio/Gameplay/Waves/Delulu_SpecterPawn_Vocalization_01
    /SpecterGameplay/Audio/Gameplay/Waves/Delulu_SpecterPawn_Vocalization_02
    /SpecterGameplay/Audio/Gameplay/Waves/Delulu_SpecterPawn_Vocalization_03
    /SpecterGameplay/Audio/Gameplay/Waves/Delulu_SpecterPawn_Vocalization_04
    /SpecterGameplay/Audio/Gameplay/Waves/Delulu_SpecterPawn_Vocalization_05
    /SpecterGameplay/Audio/Gameplay/Waves/Delulu_SpecterPawn_Vocalization_06
    /SpecterGameplay/Audio/Gameplay/Waves/Delulu_SpecterPawn_Vocalization_07
    /SpecterGameplay/Audio/Gameplay/Waves/Delulu_SpecterPawn_Vocalization_08
    /SpecterGameplay/Audio/Gameplay/Waves/Delulu_SpecterPawn_Vocalization_09
   ]
   poof [
    /SpecterGameplay/Audio/Gameplay/Waves/Delulu_Specter_UpDoot_Cloud_01
    /SpecterGameplay/Audio/Gameplay/Waves/Delulu_Specter_UpDoot_Cloud_02
    /SpecterGameplay/Audio/Gameplay/Waves/Delulu_Specter_UpDoot_Cloud_03
    /SpecterGameplay/Audio/Gameplay/Waves/Delulu_Specter_UpDoot_Cloud_04
   ]
  }
  nodes [
   TriggerDelay_0 = UE.Trigger Delay(In=^UE.Source.OnPlay) {also {defaults {Delay Time=0}}}
   ArrayRandomGet_1 = Array.Random Get:WaveAsset:Array(In Array=^ghost vocal, Next=UE.TriggerRepeat(Start=@TriggerDelay_0.Out) {also {defaults {Num Repeats=1, Period=0.1}}}.RepeatOut) {also {defaults {Weights []}}}
   n1 = UE.RandomFloat(Next=@ArrayRandomGet_1.On Next) {also {defaults {Max=8, Min=7}}}
   WavePlayer_0 = UE.Wave Player:Stereo(Wave Asset=@ArrayRandomGet_1.Value, Play=@n1.On Next, Pitch Shift=@n1.Value) {also {version [1,1]}}
   ArrayRandomGet_0 = Array.Random Get:WaveAsset:Array(Next=^UE.Source.OnPlay, In Array=^poof)
   WavePlayer_1 = UE.Wave Player:Stereo(Play=@ArrayRandomGet_0.On Next, Wave Asset=@ArrayRandomGet_0.Value) {also {version [1,1]}}
   AudioMixerAudioMixerStereo3_0 = "AudioMixer.Audio Mixer (Stereo, 3)"(In 2 L=@WavePlayer_0.Out Left, In 2 R=@WavePlayer_0.Out Right, In 0 L=@WavePlayer_1.Out Left, In 0 R=@WavePlayer_1.Out Right) {also {defaults {Gain 2=0.6}}}
  ]
  links [
   {dst="@UE.OutputFormat.Stereo.Audio:0", src=AudioMixerAudioMixerStereo3_0.Out L}
   {dst="@UE.OutputFormat.Stereo.Audio:1", src=AudioMixerAudioMixerStereo3_0.Out R}
   {dst="@UE.Source.OneShot.OnFinished", src=WavePlayer_1.On Finished}
  ]
 }]
properties {
 assetImportData=&self:AssetImportData
 attenuationSettings=/SpecterGameplay/Audio/Data/ATT_Specter_Pawn
 maxDistance=1050
 numChannels=2
 outputFormat=Stereo
 sampleRate=48000
 soundClassObject=&Fort_SFX_InGame
 volume=4
 concurrencySet [&CONC_Delulu_Specter_Ghost_Global]
}
