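# Stereo one-shot MetaSound for a Specter "Doot" hit, layering an up-doot sting with a ghost vocalization.
# On play, a fixed riser starts alongside a pitch layer shifted by a random integer amount.
# A randomly chosen vocal, triggered via delay and repeat nodes, gets a random upward pitch shift and joins the mix slightly lower.
# The riser's end finishes the source, and the `scale` input is unconnected.
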MetaSoundSource /SpecterGameplay/Audio/Gameplay/MSS_Specter_Doot_Hit
inputs [
 ghost vocal {type=WaveAsset:Array}
 scale {type=Float:Array}
 UE.Source.OnPlay {type=Trigger}
]
interfaces [UE.OutputFormat.Stereo:1.0, UE.Source.OneShot:1.0, UE.Source:1.1]
outputs [
 "UE.OutputFormat.Stereo.Audio:0" {type=Audio}
 "UE.OutputFormat.Stereo.Audio:1" {type=Audio}
 UE.Source.OneShot.OnFinished {type=Trigger}
]
pages [Default {
  input_defaults {
   UE.Source.OnPlay=null
   ghost vocal [
    /SpecterGameplay/Audio/Gameplay/Waves/Delulu_SpecterPawn_Vocalization_01
    /SpecterGameplay/Audio/Gameplay/Waves/Delulu_SpecterPawn_Vocalization_02
    /SpecterGameplay/Audio/Gameplay/Waves/Delulu_SpecterPawn_Vocalization_03
    /SpecterGameplay/Audio/Gameplay/Waves/Delulu_SpecterPawn_Vocalization_04
    /SpecterGameplay/Audio/Gameplay/Waves/Delulu_SpecterPawn_Vocalization_05
    /SpecterGameplay/Audio/Gameplay/Waves/Delulu_SpecterPawn_Vocalization_06
    /SpecterGameplay/Audio/Gameplay/Waves/Delulu_SpecterPawn_Vocalization_07
    /SpecterGameplay/Audio/Gameplay/Waves/Delulu_SpecterPawn_Vocalization_08
    /SpecterGameplay/Audio/Gameplay/Waves/Delulu_SpecterPawn_Vocalization_09
   ]
   scale [0,1,2,3,4,5,6,7,8,9,10,0]
  }
  nodes [
   n1 = UE.RandomInt32(Next=^UE.Source.OnPlay) {also {defaults {Max=6}, version [1,1]}}
   WavePlayer_1 = UE.Wave Player:Stereo(Pitch Shift=Convert.Float:Int32(Int32=@n1.Value).Float, Play=@n1.On Next) {also {defaults {Wave Asset=/SpecterGameplay/Audio/Gameplay/Waves/Delulu_Specter_UpDoot_Pitch}, version [1,1]}}
   WavePlayer_2 = UE.Wave Player:Stereo(Play=^UE.Source.OnPlay) {also {defaults {Wave Asset=/SpecterGameplay/Audio/Gameplay/Waves/Delulu_Specter_UpDoot_Riser}, version [1,1]}}
   AudioMixerAudioMixerStereo2_0 = "AudioMixer.Audio Mixer (Stereo, 2)"(In 1 L=@WavePlayer_1.Out Left, In 1 R=@WavePlayer_1.Out Right, In 0 L=@WavePlayer_2.Out Left, In 0 R=@WavePlayer_2.Out Right)
   TriggerDelay_0 = UE.Trigger Delay(In=^UE.Source.OnPlay) {also {defaults {Delay Time=0}}}
   ArrayRandomGet_0 = Array.Random Get:WaveAsset:Array(In Array=^ghost vocal, Next=UE.TriggerRepeat(Start=@TriggerDelay_0.Out) {also {defaults {Num Repeats=1, Period=0.1}}}.RepeatOut) {also {defaults {Weights []}}}
   n2 = UE.RandomFloat(Next=@ArrayRandomGet_0.On Next) {also {defaults {Max=8, Min=3}}}
   WavePlayer_0 = UE.Wave Player:Stereo(Wave Asset=@ArrayRandomGet_0.Value, Play=@n2.On Next, Pitch Shift=@n2.Value) {also {version [1,1]}}
   AudioMixerAudioMixerStereo3_0 = "AudioMixer.Audio Mixer (Stereo, 3)"(
    In 0 L=@AudioMixerAudioMixerStereo2_0.Out L
    In 0 R=@AudioMixerAudioMixerStereo2_0.Out R
    In 2 L=@WavePlayer_0.Out Left
    In 2 R=@WavePlayer_0.Out Right
   ) {also {defaults {Gain 2=0.8}}}
  ]
  links [
   {dst="@UE.OutputFormat.Stereo.Audio:0", src=AudioMixerAudioMixerStereo3_0.Out L}
   {dst="@UE.OutputFormat.Stereo.Audio:1", src=AudioMixerAudioMixerStereo3_0.Out R}
   {dst="@UE.Source.OneShot.OnFinished", src=WavePlayer_2.On Finished}
  ]
 }]
properties {
 assetImportData=&self:AssetImportData
 attenuationSettings=/SpecterGameplay/Audio/Data/ATT_Specter_Doot
 maxDistance=2500
 numChannels=2
 outputFormat=Stereo
 sampleRate=48000
 soundClassObject=&Fort_SFX_InGame
 volume=3
 concurrencySet [&CONC_Delulu_Specter_Ghost_Global]
}
