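# Mono MetaSound source for the Specter "Doot" loop, an attenuated in-game SFX that picks one of two UpDoot loop waves.
# On play, chained random nodes choose a start time, pitch shift and wave for a single Wave Player, so each trigger varies.
# The player's Loop default holds a stray MetaSoundBuilderResult string instead of a boolean, and its finish trigger drives the one-shot OnFinished output.
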MetaSoundSource /SpecterGameplay/Audio/Gameplay/MSS_Specter_Doot_Loop
inputs [
 In Array {type=WaveAsset:Array}
 UE.Source.OnPlay {type=Trigger}
]
interfaces [UE.OutputFormat.Mono:1.0, UE.Source.OneShot:1.0, UE.Source:1.1]
outputs [
 "UE.OutputFormat.Mono.Audio:0" {type=Audio}
 UE.Source.OneShot.OnFinished {type=Trigger}
]
pages [Default {
  input_defaults {
   UE.Source.OnPlay=null
   In Array [/SpecterGameplay/Audio/Gameplay/Waves/Delulu_Specter_UpDoot_Loop_01, /SpecterGameplay/Audio/Gameplay/Waves/Delulu_Specter_UpDoot_Loop_02]
  }
  nodes [
   n1 = UE.RandomTime(Next=^UE.Source.OnPlay) {also {defaults {Max=2}}}
   n2 = UE.RandomFloat(Next=@n1.On Next)
   ArrayRandomGet_0 = Array.Random Get:WaveAsset:Array(In Array=^In Array, Next=@n2.On Next)
   WavePlayer_0 = UE.Wave Player:Mono(Play=@ArrayRandomGet_0.On Next, Wave Asset=@ArrayRandomGet_0.Value, Pitch Shift=@n2.Value, Start Time=@n1.Value) {also {defaults {Loop="<MetaSoundBuilderResult.SUCCEEDED: 0>"}, version [1,1]}}
  ]
  links [{dst="@UE.Source.OneShot.OnFinished", src=WavePlayer_0.On Finished}, {dst="@UE.OutputFormat.Mono.Audio:0", src=WavePlayer_0.Out Mono}]
 }]
properties {
 assetImportData=&self:AssetImportData
 attenuationSettings=/SpecterGameplay/Audio/Data/ATT_Specter_Doot
 maxDistance=2500
 sampleRate=48000
 soundClassObject=&Fort_SFX_InGame
 volume=2.5
}
